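# Niagara swing trail for the HenNull skin of the CozyMechanic pickaxe, built from three animation-trail ribbon emitters.
# Two narrow, dual-renderer emitters fade through blue-violet curves, each writing an extra color attribute (one bright, one near-black) and animating a dynamic material parameter.
# A wider, shorter-lived emitter adds a neutral bright streak that narrows late in life.
# Base colors scale by a system time-of-day multiplier, and the first emitter's alpha also fades with emitter age.
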NiagaraSystem /BRCosmetics/Weapons/FORT_Melee/Pickaxe_CozyMechanic/Skins/HenNull/FX/NS_Pickaxe_CozyMechanic_HenNull_Trail
dependencies {
 usedDataInterfaces [&NiagaraDataInterfaceColorCurve, &NiagaraDataInterfaceCurve, &NiagaraDataInterfaceSocketReader]
 usedDynamicInputs []
 usedModules [
  &self:SystemSpawnScript.SystemScriptSource.SystemScriptGraph.NiagaraNodeAssignment_1.SetVariables_#1
  &SystemState
  &self:SystemSpawnScript.SystemScriptSource.SystemScriptGraph.NiagaraNodeAssignment_0.SetVariables_#2
 ]
}
emitters {
 FN_AnimTrail {
  emitterSpawnScript [&FN_RibbonIDOffsetUtility()]
  emitterUpdateScript [
   &EmitterState()
   &FN_SpawnAnimTrail()
  ]
  particleSpawnScript [
   &FN_InitializeAnimTrailBoneIndecies()
   &FN_InitializeAnimTrail(
    AnimTrailWidth=0.8
    Color=&MultiplyLinearColor(
     Linear Color=<rgba 1.5 1.5 1.5 1>
     Scale Factor=&MakeVector4_FromLinearColor(LinearColor=$System.TODColorMultiplier)
    )
    Lifetime=0.3
   )
  ]
  particleUpdateScript [
   &FN_UpdateAnimTrail() {module=FN_UpdateAnimTrail001}
   &ScaleColor(
    Scale Alpha=&Multiply_Float(
     A=&FloatFromCurve(FloatCurve=DataInterface() {propertyValues {Curve {keys [
          {arriveTangent=-0.4409, interpMode=RCIM_Cubic, leaveTangent=-0.4409, tangentMode=RCTM_User, time=0, value=1}
          {arriveTangent=-1.2431, interpMode=RCIM_Cubic, leaveTangent=-1.2431, tangentMode=RCTM_User, time=0.4, value=0.45}
          {arriveTangent=-0.0464, interpMode=RCIM_Cubic, tangentMode=RCTM_Break, time=1, value=0}
         ]}}})
     B=&FloatFromCurve(FloatCurve=DataInterface() {propertyValues {Curve {keys [<at 0 1>,<at 0.6 0.4>,<at 1 0>]}}}, CurveIndex=$Emitter.Age)
    )
    Scale RGB=&MakeVectorFromLinearColorRGB(Color=&ColorFromCurve(ColorCurve=DataInterface() {propertyValues {
        AlphaCurve {keys []}
        BlueCurve {keys [<at 0 3 RCIM_Cubic>,<at 0.0432 2>,<at 0.6912 0.5 RCIM_Cubic>]}
        GreenCurve {keys [<at 0 0.3 RCIM_Cubic>,<at 0.0432 0.2>,<at 0.6912 0 RCIM_Cubic>]}
        RedCurve {keys [<at 0 1.65 RCIM_Cubic>,<at 0.0432 0.8>,<at 0.6912 0.0834 RCIM_Cubic>]}
       }}))
    ScaleA=true
    ScaleRGB=true
   ) {module=ScaleColor002}
   /SetVariables(Particles.Color003:LinearColor=&MultiplyLinearColor(
     Linear Color=&MakeLinearColorFromVectorAndFloat(
      "Vector (RGB)"=&MakeVectorFromLinearColorRGB(Color=&ColorFromCurve(ColorCurve=DataInterface() {propertyValues {
          AlphaCurve {keys []}
          BlueCurve {keys [<at 0.0903 8 RCIM_Cubic>,<at 0.6287 0.5>]}
          GreenCurve {keys [<at 0.0903 0 RCIM_Cubic>,<at 0.6287 0>]}
          RedCurve {keys [<at 0.0903 2.6667 RCIM_Cubic>,<at 0.6287 0.0833>]}
         }}))
      "Float (Alpha)"=&FloatFromCurve(FloatCurve=DataInterface() {propertyValues {Curve {keys [
           <at 0 0 RCIM_Cubic>
           {arriveTangent=0.0207, interpMode=RCIM_Cubic, leaveTangent=0.0467, tangentMode=RCTM_Break, time=0.1, value=1}
           {arriveTangent=0.0276, interpMode=RCIM_Cubic, leaveTangent=0.0276, tangentMode=RCTM_User, time=0.3, value=1}
           {arriveTangent=-0.0464, interpMode=RCIM_Cubic, tangentMode=RCTM_Break, time=0.7, value=0}
          ]}}})
     )
     Scale Factor=&MakeVector4_FromLinearColor(LinearColor=$System.TODColorMultiplier)
    )) {isSetParameters=true}
   &DynamicMaterialParameters(
    Index 0 Param 1=&FloatFromCurve(FloatCurve=DataInterface() {propertyValues {Curve {keys [<at 0 2 RCIM_Cubic>,<at 1 -15 RCIM_Cubic>]}}})
    Module.Index0Float0Write=true
    Module.Index0Float1Write=true
    Module.Index0Float2Write=true
    Module.Index0Float3Write=true
   )
  ]
  rendererClasses [&NiagaraRibbonRendererProperties]
  renderers [
   {rendererClass=&NiagaraRibbonRendererProperties}
   {rendererClass=&NiagaraRibbonRendererProperties, rendererIndex=1}
  ]
 }
 FN_AnimTrail001 {
  emitterSpawnScript [&FN_RibbonIDOffsetUtility()]
  emitterUpdateScript [
   &EmitterState()
   &FN_SpawnAnimTrail()
  ]
  particleSpawnScript [
   &FN_InitializeAnimTrailBoneIndecies()
   &FN_InitializeAnimTrail(
    AnimTrailWidth=0.8
    Color=&MultiplyLinearColor(
     Linear Color=<rgba 1.5 1.5 1.5 1>
     Scale Factor=&MakeVector4_FromLinearColor(LinearColor=$System.TODColorMultiplier)
    )
    Lifetime=0.25
   )
  ]
  particleUpdateScript [
   &FN_UpdateAnimTrail() {module=FN_UpdateAnimTrail001}
   /SetVariables(Particles.DarkColor:LinearColor=&MultiplyLinearColor(
     Linear Color=&MakeLinearColorFromVectorAndFloat(
      "Vector (RGB)"=&MakeVectorFromLinearColorRGB(Color=&ColorFromCurve(ColorCurve=DataInterface() {propertyValues {
          AlphaCurve {keys []}
          BlueCurve {keys [<at 0 0.1 RCIM_Cubic>,<at 0.2586 0.01 RCIM_Cubic>]}
          GreenCurve {keys [<at 0 0.01 RCIM_Cubic>,<at 0.2586 0 RCIM_Cubic>]}
          RedCurve {keys [<at 0 0.04 RCIM_Cubic>,<at 0.2586 0.0017 RCIM_Cubic>]}
         }}))
      "Float (Alpha)"=&FloatFromCurve(
       FloatCurve=DataInterface() {propertyValues {Curve {keys [
           {arriveTangent=-0.0791, interpMode=RCIM_Cubic, leaveTangent=-0.0791, tangentMode=RCTM_User, time=0, value=1}
           {arriveTangent=-2.4062, interpMode=RCIM_Cubic, leaveTangent=-2.4062, tangentMode=RCTM_User, time=0.3586, value=0.5749}
           <at 0.6 0 RCIM_Cubic>
          ]}}}
       Scale Curve=3
      )
     )
     Scale Factor=&MakeVector4_FromLinearColor(LinearColor=$System.TODColorMultiplier)
    )) {isSetParameters=true}
   &ScaleColor(
    Scale Alpha=&FloatFromCurve(FloatCurve=DataInterface() {propertyValues {Curve {keys [
         {arriveTangent=-0.4618, interpMode=RCIM_Cubic, leaveTangent=-0.4618, tangentMode=RCTM_User, time=0, value=1}
         {arriveTangent=-1.1299, interpMode=RCIM_Cubic, leaveTangent=-1.1299, tangentMode=RCTM_User, time=0.4298, value=0.36}
         {arriveTangent=-0.0168, interpMode=RCIM_Cubic, tangentMode=RCTM_Break, time=1, value=0}
        ]}}})
    Scale RGB=&MakeVectorFromLinearColorRGB(Color=&ColorFromCurve(ColorCurve=DataInterface() {propertyValues {
        AlphaCurve {keys []}
        BlueCurve {keys [<at 0 3 RCIM_Cubic>,<at 0.0432 2>,<at 0.6912 0.5 RCIM_Cubic>]}
        GreenCurve {keys [<at 0 0.3 RCIM_Cubic>,<at 0.0432 0.2>,<at 0.6912 0 RCIM_Cubic>]}
        RedCurve {keys [<at 0 1.65 RCIM_Cubic>,<at 0.0432 0.8>,<at 0.6912 0.0833 RCIM_Cubic>]}
       }}))
    ScaleA=true
    ScaleRGB=true
   ) {module=ScaleColor002}
   &DynamicMaterialParameters(
    Index 0 Param 1=&FloatFromCurve(FloatCurve=DataInterface() {propertyValues {Curve {keys [<at 0 2 RCIM_Cubic>,<at 1 -7 RCIM_Cubic>]}}})
    Module.Index0Float0Write=true
    Module.Index0Float1Write=true
    Module.Index0Float2Write=true
    Module.Index0Float3Write=true
   )
  ]
  rendererClasses [&NiagaraRibbonRendererProperties]
  renderers [
   {rendererClass=&NiagaraRibbonRendererProperties}
   {rendererClass=&NiagaraRibbonRendererProperties, rendererIndex=1}
  ]
 }
 FN_AnimTrail002 {
  emitterSpawnScript [&FN_RibbonIDOffsetUtility()]
  emitterUpdateScript [
   &EmitterState()
   &FN_SpawnAnimTrail()
  ]
  particleSpawnScript [
   &FN_InitializeAnimTrailBoneIndecies()
   &FN_InitializeAnimTrail(
    AnimTrailWidth=1.5
    Color=&MultiplyLinearColor(
     Linear Color=<rgba 2 2 2 1.2>
     Scale Factor=&MakeVector4_FromLinearColor(LinearColor=$System.TODColorMultiplier)
    )
    Lifetime=0.15
   )
  ]
  particleUpdateScript [
   &FN_UpdateAnimTrail() {module=FN_UpdateAnimTrail001}
   &ScaleColor(
    Scale Alpha=&FloatFromCurve(FloatCurve=DataInterface() {propertyValues {Curve {keys [
         {arriveTangent=7.8189, interpMode=RCIM_Cubic, leaveTangent=7.8189, tangentMode=RCTM_User, time=0, value=0.2}
         {arriveTangent=0.0168, interpMode=RCIM_Cubic, leaveTangent=0.0168, tangentMode=RCTM_User, time=0.12, value=1}
         <at 0.8 0 RCIM_Cubic>
        ]}}})
    Scale RGB=&Multiply_VectorByFloat(
     Vector=<1 1 1>
     Float=&FloatFromCurve(FloatCurve=DataInterface() {propertyValues {Curve {keys [
          {interpMode=RCIM_Cubic, tangentMode=RCTM_User, time=0, value=3}
          {arriveTangent=-2.7567, interpMode=RCIM_Cubic, leaveTangent=-2.7567, tangentMode=RCTM_User, time=0.4, value=1.75}
          {interpMode=RCIM_Cubic, tangentMode=RCTM_User, time=1, value=1}
         ]}}})
    )
    ScaleA=true
    ScaleRGB=true
   ) {module=ScaleColor002}
   &ScaleRibbonWidth(Ribbon Width Scale=&FloatFromCurve(FloatCurve=DataInterface() {propertyValues {Curve {keys [<at 0.7 1>,<at 1 0.5>]}}}))
  ]
  rendererClasses [&NiagaraRibbonRendererProperties]
  renderers [{rendererClass=&NiagaraRibbonRendererProperties}]
 }
}
summary {
 systemName=NS_Pickaxe_CozyMechanic_HenNull_Trail
 emitters [
  {emitterName=FN_AnimTrail, rendererClasses [&NiagaraRibbonRendererProperties]}
  {emitterName=FN_AnimTrail001, rendererClasses [&NiagaraRibbonRendererProperties]}
  {emitterName=FN_AnimTrail002, rendererClasses [&NiagaraRibbonRendererProperties]}
 ]
 userVariables [
  User.bIsFrontEnd {type=NiagaraBool}
  User.bIsFrontEndPreview {type=NiagaraBool}
  User.RibbonIDOffset {type=NiagaraInt32}
 ]
}
